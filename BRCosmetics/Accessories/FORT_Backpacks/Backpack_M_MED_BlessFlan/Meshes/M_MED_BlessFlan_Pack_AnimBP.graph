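# Animation Blueprint for the BlessFlan backpack cosmetic that adjusts the pack bone and simulates its dangling straps.
# The EventGraph maps spine bend to additive pack offsets, derives an airborne flag and flags parachute closing for three seconds.
# AnimDynamics chains, tuned separately for front end, airborne, crouched and default states, drive both straps and are selected by bool blends.
# Skydive, dive and parachute-close states add blended pack nudges.
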AnimBlueprint /BRCosmetics/Accessories/FORT_Backpacks/Backpack_M_MED_BlessFlan/Meshes/M_MED_BlessFlan_Pack_AnimBP
target_skeleton=&/M_MED_BlessFlan_Pack_Skeleton
graphs [
 AnimGraph {nodes [
   n1 = Anim Dynamics - Bone: dyn_strap_base!(
    ComponentPose=Local To Component(LocalPose=Use cached pose 'Base_pose'())
    bAlphaBoolEnabled:bool=True
    GravityOverride:Vector="0,0,-980"
    ExternalForce="0,0,0"
   ) -> (Pose) {
    alphaInputType=Bool
    angularDampingOverride=0.9999
    angularSpringConstant=1000
    bAngularSpring=true
    bOverrideAngularDamping=true
    bUseGravityOverride=true
    simulationSpace=RootRelative
    alphaBoolBlend {blendInTime=0.1, blendOption=Linear, blendOutTime=0.1}
    boundBone {boneName=dyn_strap_base}
    componentAppliedLinearAccClamp=<3000 3000 3000>
    componentLinearAccScale=<0.4 0.5 0.5>
    componentLinearVelScale=<0.5 0.5 0.5>
    gravityOverride {z=-500}
    physicsBodyDefinitions [{
      collisionType=CustomSphere
      sphereCollisionRadius=2.5
      boundBone {boneName=dyn_strap_base}
      boxExtents=<12 2 6>
      constraintSetup {angularLimitsMax=<0 10 20>, angularLimitsMin=<0 -10 -55>, angularTarget=<0 0 0>}
      localJointOffset=<1.5 0 0>
     }]
    relativeSpaceBone {boneName=pack}
   }
   n2 = Anim Dynamics - Bone: dyn_strap_01!(ComponentPose=@n1, bAlphaBoolEnabled:bool=True, GravityOverride:Vector="0,0,-980", ExternalForce="0,0,0") -> (Pose) {
    alphaInputType=Bool
    angularDampingOverride=0.9999
    angularSpringConstant=1000
    bAngularSpring=true
    bOverrideAngularDamping=true
    bUseGravityOverride=true
    simulationSpace=RootRelative
    alphaBoolBlend {blendInTime=0.1, blendOption=Linear, blendOutTime=0.1}
    boundBone {boneName=dyn_strap_01}
    componentAppliedLinearAccClamp=<3000 3000 3000>
    componentLinearAccScale=<0.4 0.5 0.5>
    componentLinearVelScale=<0.5 0.5 0.5>
    gravityOverride {z=-500}
    physicsBodyDefinitions [{
      collisionType=CustomSphere
      sphereCollisionRadius=2.5
      boundBone {boneName=dyn_strap_01}
      boxExtents=<12 2 6>
      constraintSetup {angularLimitsMax=<0 20 5>, angularLimitsMin=<0 -20 -55>, angularTarget=<0 0 0>}
      localJointOffset=<1.5 0 0>
     }]
    relativeSpaceBone {boneName=pack}
   }
   n3 = Anim Dynamics - Bone: dyn_strap_02!(ComponentPose=@n2, bAlphaBoolEnabled:bool=True, GravityOverride:Vector="0,0,-980", ExternalForce="0,0,0") -> (Pose) {
    alphaInputType=Bool
    angularDampingOverride=0.9999
    angularSpringConstant=1000
    bAngularSpring=true
    bOverrideAngularDamping=true
    bUseGravityOverride=true
    simulationSpace=RootRelative
    alphaBoolBlend {blendInTime=0.1, blendOption=Linear, blendOutTime=0.1}
    boundBone {boneName=dyn_strap_02}
    componentAppliedLinearAccClamp=<3000 3000 3000>
    componentLinearAccScale=<0.4 0.5 0.5>
    componentLinearVelScale=<0.5 0.5 0.5>
    gravityOverride {z=-500}
    physicsBodyDefinitions [{
      collisionType=CustomSphere
      sphereCollisionRadius=2.5
      boundBone {boneName=dyn_strap_02}
      boxExtents=<12 2 6>
      constraintSetup {angularLimitsMax=<0 10 5>, angularLimitsMin=<0 -10 -55>, angularTarget=<0 0 0>}
      localJointOffset=<1.5 0 0>
     }]
    relativeSpaceBone {boneName=pack}
   }
   Blend Poses by bool(BlendPose_0=Component To Local(ComponentPose=@n3), bActiveValue=Get bIsCrouching()) {blendType=Linear}
   n4 = "Transform (Modify) Bone - Bone: pack"(ComponentPose=Local To Component(), Translation=Get backpack_translation(), Rotation=Get backpack_rotate()) {
    rotationMode=BMM_Additive
    scaleMode=BMM_Additive
    scaleSpace=BCS_BoneSpace
    translationMode=BMM_Additive
    alphaBoolBlend {blendOption=Linear}
    boneToModify {boneName=pack}
   }
   Component To Local(ComponentPose=@n4)
   n5 = Anim Dynamics - Bone: dyn_strap_base(
    ComponentPose=Local To Component(LocalPose=Use cached pose 'Base_pose'())
    bAlphaBoolEnabled=True
    GravityOverride="0,0,-980"
    ExternalForce="0,0,0"
   ) {
    alphaInputType=Bool
    angularDampingOverride=0.9999
    angularSpringConstant=1000
    bAngularSpring=true
    bOverrideAngularDamping=true
    bUseGravityOverride=true
    simulationSpace=RootRelative
    alphaBoolBlend {blendInTime=0.1, blendOption=Linear, blendOutTime=0.1}
    boundBone {boneName=dyn_strap_base}
    componentAppliedLinearAccClamp=<1500 1500 1500>
    componentLinearAccScale=<0.4 0.5 0.5>
    componentLinearVelScale=<0.5 0.5 0.5>
    gravityOverride {z=-500}
    physicsBodyDefinitions [{
      collisionType=CustomSphere
      sphereCollisionRadius=2.5
      boundBone {boneName=dyn_strap_base}
      boxExtents=<12 2 6>
      constraintSetup {angularLimitsMax=<0 10 5>, angularLimitsMin=<0 -10 -55>, angularTarget=<0 0 0>}
      localJointOffset=<1.5 0 0>
     }]
    relativeSpaceBone {boneName=pack}
   }
   n6 = Anim Dynamics - Bone: dyn_strap_01(ComponentPose=@n5, bAlphaBoolEnabled=True, GravityOverride="0,0,-980", ExternalForce="0,0,0") {
    alphaInputType=Bool
    angularDampingOverride=0.9999
    angularSpringConstant=1000
    bAngularSpring=true
    bOverrideAngularDamping=true
    bUseGravityOverride=true
    simulationSpace=RootRelative
    alphaBoolBlend {blendInTime=0.1, blendOption=Linear, blendOutTime=0.1}
    boundBone {boneName=dyn_strap_01}
    componentAppliedLinearAccClamp=<1500 1500 1500>
    componentLinearAccScale=<0.4 0.5 0.5>
    componentLinearVelScale=<0.5 0.5 0.5>
    gravityOverride {z=-500}
    physicsBodyDefinitions [{
      collisionType=CustomSphere
      sphereCollisionRadius=2.5
      boundBone {boneName=dyn_strap_01}
      boxExtents=<12 2 6>
      constraintSetup {angularLimitsMax=<0 20 5>, angularLimitsMin=<0 -20 -55>, angularTarget=<0 0 0>}
      localJointOffset=<1.5 0 0>
     }]
    relativeSpaceBone {boneName=pack}
   }
   n7 = Anim Dynamics - Bone: dyn_strap_02(ComponentPose=@n6, bAlphaBoolEnabled=True, GravityOverride="0,0,-980", ExternalForce="0,0,0") {
    alphaInputType=Bool
    angularDampingOverride=0.9999
    angularSpringConstant=1000
    bAngularSpring=true
    bOverrideAngularDamping=true
    bUseGravityOverride=true
    simulationSpace=RootRelative
    alphaBoolBlend {blendInTime=0.1, blendOption=Linear, blendOutTime=0.1}
    boundBone {boneName=dyn_strap_02}
    componentAppliedLinearAccClamp=<1500 1500 1500>
    componentLinearAccScale=<0.4 0.5 0.5>
    componentLinearVelScale=<0.5 0.5 0.5>
    gravityOverride {z=-500}
    physicsBodyDefinitions [{
      collisionType=CustomSphere
      sphereCollisionRadius=2.5
      boundBone {boneName=dyn_strap_02}
      boxExtents=<12 2 6>
      constraintSetup {angularLimitsMax=<0 10 5>, angularLimitsMin=<0 -10 -55>, angularTarget=<0 0 0>}
      localJointOffset=<1.5 0 0>
     }]
    relativeSpaceBone {boneName=pack}
   }
   n8 = Anim Dynamics - Bone: dyn_strap_small_base!(ComponentPose=@n7, bAlphaBoolEnabled:bool=True, GravityOverride:Vector="0,0,-980", ExternalForce="0,0,0") -> (Pose) {
    alphaInputType=Bool
    angularDampingOverride=0.9999
    angularSpringConstant=1000
    bAngularSpring=true
    bOverrideAngularDamping=true
    bUseGravityOverride=true
    simulationSpace=RootRelative
    alphaBoolBlend {blendInTime=0.1, blendOption=Linear, blendOutTime=0.1}
    boundBone {boneName=dyn_strap_small_base}
    componentAppliedLinearAccClamp=<1500 1500 1500>
    componentLinearAccScale=<0.4 0.5 0.5>
    componentLinearVelScale=<0.5 0.5 0.5>
    gravityOverride {z=-500}
    physicsBodyDefinitions [{
      collisionType=CustomSphere
      sphereCollisionRadius=2.5
      boundBone {boneName=dyn_strap_small_base}
      boxExtents=<10 3 5>
      constraintSetup {angularLimitsMax=<15 10 4>, angularLimitsMin=<-15 -10 -45>, angularTarget=<0 0 0>}
      localJointOffset=<1.5 0 0>
     }]
    relativeSpaceBone {boneName=pack}
   }
   n9 = Anim Dynamics - Bone: dyn_strap_small_01!(ComponentPose=@n8, bAlphaBoolEnabled:bool=True, GravityOverride:Vector="0,0,-980", ExternalForce="0,0,0") -> (Pose) {
    alphaInputType=Bool
    angularDampingOverride=0.9999
    angularSpringConstant=1000
    bAngularSpring=true
    bOverrideAngularDamping=true
    bUseGravityOverride=true
    simulationSpace=RootRelative
    alphaBoolBlend {blendInTime=0.1, blendOption=Linear, blendOutTime=0.1}
    boundBone {boneName=dyn_strap_small_01}
    componentAppliedLinearAccClamp=<1500 1500 1500>
    componentLinearAccScale=<0.4 0.5 0.5>
    componentLinearVelScale=<0.5 0.5 0.5>
    gravityOverride {z=-500}
    physicsBodyDefinitions [{
      collisionType=CustomSphere
      sphereCollisionRadius=2.5
      boundBone {boneName=dyn_strap_small_01}
      boxExtents=<10 3 5>
      constraintSetup {angularLimitsMax=<15 15 4>, angularLimitsMin=<-15 -15 -55>, angularTarget=<0 0 0>}
      localJointOffset=<1.5 0 0>
     }]
    relativeSpaceBone {boneName=pack}
   }
   n10 = "Transform (Modify) Bone - Bone: pack"!(
    ComponentPose=Local To Component(LocalPose=Use cached pose 'Base_pose'())
    bAlphaBoolEnabled:bool=Get bIsSkyDiving()
    Translation="0,-2,8"
    Rotation="0,0,15"
    Scale
   ) -> (Pose) {
    alphaInputType=Bool
    rotationMode=BMM_Replace
    rotationSpace=BCS_BoneSpace
    translationMode=BMM_Additive
    translationSpace=BCS_BoneSpace
    alphaBoolBlend {blendInTime=0.2, blendOption=Linear, blendOutTime=0.2}
    boneToModify {boneName=pack}
    rotation {roll=15}
    translation {y=-2, z=8}
   }
   n11 = Anim Dynamics - Bone: dyn_strap_base(ComponentPose=@n10, bAlphaBoolEnabled=True, GravityOverride="0,0,-980", ExternalForce="0,0,0") {
    alphaInputType=Bool
    angularDampingOverride=0.9999
    angularSpringConstant=1000
    bOverrideAngularDamping=true
    bUseGravityOverride=true
    simulationSpace=BoneRelative
    alphaBoolBlend {blendInTime=0.1, blendOption=Linear, blendOutTime=0.1}
    boundBone {boneName=dyn_strap_base}
    componentAppliedLinearAccClamp=<1500 1500 1500>
    componentLinearAccScale=<0.4 0.5 0.3>
    componentLinearVelScale=<0.5 0.5 0.3>
    gravityOverride {z=-500}
    physicsBodyDefinitions [{
      collisionType=CustomSphere
      sphereCollisionRadius=2.5
      boundBone {boneName=dyn_strap_base}
      boxExtents=<12 5 7>
      constraintSetup {angularLimitsMax=<15 10 4>, angularLimitsMin=<-15 -10 -40>, angularTarget=<0 0 0>}
      localJointOffset=<2 0 0>
     }]
    relativeSpaceBone {boneName=pack}
   }
   n12 = Anim Dynamics - Bone: dyn_strap_01(ComponentPose=@n11, bAlphaBoolEnabled=True, GravityOverride="0,0,-980", ExternalForce="0,0,0") {
    alphaInputType=Bool
    angularDampingOverride=0.9999
    angularSpringConstant=1000
    bOverrideAngularDamping=true
    bUseGravityOverride=true
    simulationSpace=BoneRelative
    alphaBoolBlend {blendInTime=0.1, blendOption=Linear, blendOutTime=0.1}
    boundBone {boneName=dyn_strap_01}
    componentAppliedLinearAccClamp=<1500 1500 1500>
    componentLinearAccScale=<0.4 0.5 0.3>
    componentLinearVelScale=<0.5 0.5 0.3>
    gravityOverride {z=-500}
    physicsBodyDefinitions [{
      collisionType=CustomSphere
      sphereCollisionRadius=2.5
      boundBone {boneName=dyn_strap_01}
      boxExtents=<12 5 7>
      constraintSetup {angularLimitsMax=<15 15 4>, angularLimitsMin=<-15 -15 -50>, angularTarget=<0 0 0>}
      localJointOffset=<2 0 0>
     }]
    relativeSpaceBone {boneName=pack}
   }
   n13 = Anim Dynamics - Bone: dyn_strap_02(ComponentPose=@n12, bAlphaBoolEnabled=True, GravityOverride="0,0,-980", ExternalForce="0,0,0") {
    alphaInputType=Bool
    angularDampingOverride=0.9999
    angularSpringConstant=1000
    bOverrideAngularDamping=true
    bUseGravityOverride=true
    simulationSpace=BoneRelative
    alphaBoolBlend {blendInTime=0.1, blendOption=Linear, blendOutTime=0.1}
    boundBone {boneName=dyn_strap_02}
    componentAppliedLinearAccClamp=<1500 1500 1500>
    componentLinearAccScale=<0.4 0.5 0.3>
    componentLinearVelScale=<0.5 0.5 0.3>
    gravityOverride {z=-500}
    physicsBodyDefinitions [{
      collisionType=CustomSphere
      sphereCollisionRadius=2.5
      boundBone {boneName=dyn_strap_02}
      boxExtents=<6 3 3.5>
      constraintSetup {angularLimitsMax=<15 10 4>, angularLimitsMin=<-15 -10 -60>, angularTarget=<0 0 0>}
      localJointOffset=<2 0 0>
     }]
    relativeSpaceBone {boneName=pack}
   }
   n14 = Anim Dynamics - Bone: dyn_strap_small_base(ComponentPose=@n13, bAlphaBoolEnabled=True, GravityOverride="0,0,-980", ExternalForce="0,0,0") {
    alphaInputType=Bool
    angularDampingOverride=0.9999
    angularSpringConstant=1000
    bOverrideAngularDamping=true
    bUseGravityOverride=true
    simulationSpace=BoneRelative
    alphaBoolBlend {blendInTime=0.1, blendOption=Linear, blendOutTime=0.1}
    boundBone {boneName=dyn_strap_small_base}
    componentAppliedLinearAccClamp=<2000 2000 2000>
    componentLinearAccScale=<0.4 0.5 0.3>
    componentLinearVelScale=<0.5 0.5 0.3>
    gravityOverride {z=-500}
    physicsBodyDefinitions [{
      collisionType=CustomSphere
      sphereCollisionRadius=2.5
      boundBone {boneName=dyn_strap_small_base}
      boxExtents=<10 3 5>
      constraintSetup {angularLimitsMax=<15 10 4>, angularLimitsMin=<-15 -10 -45>, angularTarget=<0 0 0>}
      localJointOffset=<2 0 0>
     }]
    relativeSpaceBone {boneName=pack}
   }
   n15 = Anim Dynamics - Bone: dyn_strap_small_01(ComponentPose=@n14, bAlphaBoolEnabled=True, GravityOverride="0,0,-980", ExternalForce="0,0,0") {
    alphaInputType=Bool
    angularDampingOverride=0.9999
    angularSpringConstant=1000
    bOverrideAngularDamping=true
    bUseGravityOverride=true
    simulationSpace=BoneRelative
    alphaBoolBlend {blendInTime=0.1, blendOption=Linear, blendOutTime=0.1}
    boundBone {boneName=dyn_strap_small_01}
    componentAppliedLinearAccClamp=<2000 2000 2000>
    componentLinearAccScale=<0.4 0.5 0.3>
    componentLinearVelScale=<0.5 0.5 0.3>
    gravityOverride {z=-500}
    physicsBodyDefinitions [{
      collisionType=CustomSphere
      sphereCollisionRadius=2.5
      boundBone {boneName=dyn_strap_small_01}
      boxExtents=<10 3 5>
      constraintSetup {angularLimitsMax=<15 15 4>, angularLimitsMin=<-15 -15 -55>, angularTarget=<0 0 0>}
      localJointOffset=<2 0 0>
     }]
    relativeSpaceBone {boneName=pack}
   }
   n16 = "Transform (Modify) Bone - Bone: pack"(ComponentPose=@n15, bAlphaBoolEnabled=Get bIsSkydiveDiveMode(), Translation="0,3,-3", Rotation="0,0,-15") {
    alphaInputType=Bool
    rotationMode=BMM_Additive
    rotationSpace=BCS_BoneSpace
    translationMode=BMM_Additive
    translationSpace=BCS_BoneSpace
    alphaBoolBlend {blendInTime=0.1, blendOption=Linear, blendOutTime=0.3}
    alphaScaleBiasClamp {bInterpResult=true, interpSpeedDecreasing=1, interpSpeedIncreasing=1}
    boneToModify {boneName=pack}
    rotation {roll=-15}
    translation {y=3, z=-3}
   }
   n17 = Blend Poses by bool(BlendPose_0=Component To Local(ComponentPose=@n9), BlendPose_1=Component To Local(ComponentPose=@n16), bActiveValue=Get bIsCrouching()) {blendType=Linear}
   n18 = "Transform (Modify) Bone - Bone: pack"!(
    ComponentPose=Local To Component(LocalPose=Copy Pose From Mesh())
    Alpha:float
    Translation=Get backpack_translation()
    Rotation=Get backpack_rotate()
    Scale
   ) -> (Pose) {
    rotationMode=BMM_Additive
    scaleMode=BMM_Additive
    scaleSpace=BCS_BoneSpace
    translationMode=BMM_Additive
    alphaBoolBlend {blendOption=Linear}
    boneToModify {boneName=pack}
   }
   n19 = Anim Dynamics - Bone: dyn_strap_01(ComponentPose=@n18, bAlphaBoolEnabled=True, GravityOverride="0,0,-700", ExternalForce="0,0,0") {
    alphaInputType=Bool
    angularDampingOverride=0.9999
    angularSpringConstant=1000
    bAngularSpring=true
    bOverrideAngularDamping=true
    bUseGravityOverride=true
    alphaBoolBlend {blendInTime=0.1, blendOption=Linear, blendOutTime=0.1}
    boundBone {boneName=dyn_strap_01}
    componentAppliedLinearAccClamp=<1500 1500 15000>
    componentLinearAccScale=<0.4 0.5 0.5>
    componentLinearVelScale=<0.5 0.5 0.5>
    gravityOverride {z=-500}
    physicsBodyDefinitions [{
      collisionType=CustomSphere
      sphereCollisionRadius=2.5
      boundBone {boneName=dyn_strap_01}
      boxExtents=<10 10 10>
      constraintSetup {angularLimitsMax=<0 10 -5>, angularLimitsMin=<0 -10 -70>, angularTarget=<0 0 0>}
      localJointOffset=<1.5 0 0>
     }]
    relativeSpaceBone {boneName=pack}
   }
   n20 = Anim Dynamics - Bone: dyn_strap_02(ComponentPose=@n19, bAlphaBoolEnabled=True, GravityOverride="0,0,-700", ExternalForce="0,0,0") {
    alphaInputType=Bool
    angularDampingOverride=0.9999
    angularSpringConstant=1000
    bAngularSpring=true
    bOverrideAngularDamping=true
    bUseGravityOverride=true
    alphaBoolBlend {blendInTime=0.1, blendOption=Linear, blendOutTime=0.1}
    boundBone {boneName=dyn_strap_02}
    componentAppliedLinearAccClamp=<1500 1500 15000>
    componentLinearAccScale=<0.4 0.5 0.5>
    componentLinearVelScale=<0.5 0.5 0.5>
    gravityOverride {z=-500}
    physicsBodyDefinitions [{
      collisionType=CustomSphere
      sphereCollisionRadius=2.5
      boundBone {boneName=dyn_strap_02}
      boxExtents=<10 10 10>
      constraintSetup {angularLimitsMax=<0 10 -5>, angularLimitsMin=<0 -10 -70>, angularTarget=<0 0 0>}
      localJointOffset=<1.5 0 0>
     }]
    relativeSpaceBone {boneName=pack}
   }
   n21 = Anim Dynamics - Bone: dyn_strap_small_base(ComponentPose=@n20, bAlphaBoolEnabled=True, GravityOverride="0,0,-700", ExternalForce="0,0,0") {
    alphaInputType=Bool
    angularDampingOverride=0.9999
    angularSpringConstant=1000
    bOverrideAngularDamping=true
    bUseGravityOverride=true
    simulationSpace=BoneRelative
    alphaBoolBlend {blendInTime=0.1, blendOption=Linear, blendOutTime=0.1}
    boundBone {boneName=dyn_strap_small_base}
    componentAppliedLinearAccClamp=<2000 2000 2000>
    componentLinearAccScale=<0.4 0.5 0.3>
    componentLinearVelScale=<0.5 0.5 0.3>
    gravityOverride {z=-500}
    physicsBodyDefinitions [{
      collisionType=CustomSphere
      sphereCollisionRadius=2.5
      boundBone {boneName=dyn_strap_small_base}
      boxExtents=<10 3 5>
      constraintSetup {angularLimitsMax=<15 10 -6>, angularLimitsMin=<-15 -10 -75>, angularTarget=<0 0 0>}
      localJointOffset=<2 0 0>
     }]
    relativeSpaceBone {boneName=pack}
   }
   n22 = Anim Dynamics - Bone: dyn_strap_small_01(ComponentPose=@n21, bAlphaBoolEnabled=True, GravityOverride="0,0,-700", ExternalForce="0,0,0") {
    alphaInputType=Bool
    angularDampingOverride=0.9999
    angularSpringConstant=1000
    bOverrideAngularDamping=true
    bUseGravityOverride=true
    simulationSpace=BoneRelative
    alphaBoolBlend {blendInTime=0.1, blendOption=Linear, blendOutTime=0.1}
    boundBone {boneName=dyn_strap_small_01}
    componentAppliedLinearAccClamp=<2000 2000 2000>
    componentLinearAccScale=<0.4 0.5 0.3>
    componentLinearVelScale=<0.5 0.5 0.3>
    gravityOverride {z=-500}
    physicsBodyDefinitions [{
      collisionType=CustomSphere
      sphereCollisionRadius=2.5
      boundBone {boneName=dyn_strap_small_01}
      boxExtents=<10 3 5>
      constraintSetup {angularLimitsMax=<15 15 -6>, angularLimitsMin=<-15 -15 -75>, angularTarget=<0 0 0>}
      localJointOffset=<2 0 0>
     }]
    relativeSpaceBone {boneName=pack}
   }
   n23 = "Transform (Modify) Bone - Bone: pack"!(
    ComponentPose=@n22
    bAlphaBoolEnabled:bool=Get Parashute Closed() -> (Parashute Closed:bool)
    Translation="0,-4,0"
    Rotation="0,-5,-10"
    Scale
   ) -> (Pose) {
    alphaInputType=Bool
    rotationMode=BMM_Additive
    rotationSpace=BCS_BoneSpace
    translationMode=BMM_Additive
    translationSpace=BCS_BoneSpace
    alphaBoolBlend {blendInTime=0.1, blendOption=Linear, blendOutTime=0.3}
    alphaScaleBiasClamp {bInterpResult=true, interpSpeedDecreasing=1, interpSpeedIncreasing=1}
    boneToModify {boneName=pack}
    rotation {roll=-10, yaw=-5}
    translation {y=-4}
   }
   n24 = Component To Local(ComponentPose=@n23)
   n25 = Get inAir() -> (inAir:bool)
   n26 = Anim Dynamics - Bone: dyn_strap_base(
    ComponentPose=Local To Component(LocalPose=Use cached pose 'Base_pose'())
    bAlphaBoolEnabled=True
    GravityOverride="0,0,-980"
    ExternalForce="0,0,0"
   ) {
    alphaInputType=Bool
    angularDampingOverride=1
    angularSpringConstant=1000
    bAngularSpring=true
    bOverrideAngularDamping=true
    bUseGravityOverride=true
    simulationSpace=RootRelative
    alphaBoolBlend {blendInTime=0.1, blendOption=Linear, blendOutTime=0.1}
    boundBone {boneName=dyn_strap_base}
    componentAppliedLinearAccClamp=<1000 1000 1000>
    componentLinearAccScale=<0.3 0.3 0.3>
    componentLinearVelScale=<0.4 0.4 0.4>
    gravityOverride {z=-500}
    physicsBodyDefinitions [{
      collisionType=CustomSphere
      sphereCollisionRadius=2.5
      boundBone {boneName=dyn_strap_base}
      boxExtents=<7 3 5>
      constraintSetup {angularLimitsMax=<0 15 4>, angularLimitsMin=<0 -15 0>, angularTarget=<0 0 0>}
      localJointOffset=<0.8 0 0>
     }]
    relativeSpaceBone {boneName=pack}
   }
   n27 = Anim Dynamics - Bone: dyn_strap_01(ComponentPose=@n26, bAlphaBoolEnabled=True, GravityOverride="0,0,-980", ExternalForce="0,0,0") {
    alphaInputType=Bool
    angularDampingOverride=1
    angularSpringConstant=1000
    bAngularSpring=true
    bOverrideAngularDamping=true
    bUseGravityOverride=true
    simulationSpace=RootRelative
    alphaBoolBlend {blendInTime=0.1, blendOption=Linear, blendOutTime=0.1}
    boundBone {boneName=dyn_strap_01}
    componentAppliedLinearAccClamp=<1000 1000 1000>
    componentLinearAccScale=<0.3 0.3 0.3>
    componentLinearVelScale=<0.4 0.4 0.4>
    gravityOverride {z=-500}
    physicsBodyDefinitions [{
      collisionType=CustomSphere
      sphereCollisionRadius=2.5
      boundBone {boneName=dyn_strap_01}
      boxExtents=<7 3 5>
      constraintSetup {angularLimitsMax=<0 20 4>, angularLimitsMin=<0 -20 -20>, angularTarget=<0 0 0>}
      localJointOffset=<0.8 0 0>
     }]
    relativeSpaceBone {boneName=pack}
   }
   n28 = Anim Dynamics - Bone: dyn_strap_02(ComponentPose=@n27, bAlphaBoolEnabled=True, GravityOverride="0,0,-980", ExternalForce="0,0,0") {
    alphaInputType=Bool
    angularDampingOverride=1
    angularSpringConstant=1000
    bAngularSpring=true
    bOverrideAngularDamping=true
    bUseGravityOverride=true
    simulationSpace=RootRelative
    alphaBoolBlend {blendInTime=0.1, blendOption=Linear, blendOutTime=0.1}
    boundBone {boneName=dyn_strap_02}
    componentAppliedLinearAccClamp=<1000 1000 1000>
    componentLinearAccScale=<0.3 0.3 0.3>
    componentLinearVelScale=<0.4 0.4 0.4>
    gravityOverride {z=-500}
    physicsBodyDefinitions [{
      collisionType=CustomSphere
      sphereCollisionRadius=2.5
      boundBone {boneName=dyn_strap_02}
      boxExtents=<7 3 5>
      constraintSetup {angularLimitsMax=<0 20 4>, angularLimitsMin=<0 -20 -20>, angularTarget=<0 0 0>}
      localJointOffset=<0.8 0 0>
     }]
    relativeSpaceBone {boneName=pack}
   }
   n29 = Anim Dynamics - Bone: dyn_strap_small_base(ComponentPose=@n28, bAlphaBoolEnabled=True, GravityOverride="0,0,-980", ExternalForce="0,0,0") {
    alphaInputType=Bool
    angularDampingOverride=1
    angularSpringConstant=1000
    bAngularSpring=true
    bOverrideAngularDamping=true
    bUseGravityOverride=true
    simulationSpace=RootRelative
    alphaBoolBlend {blendInTime=0.1, blendOption=Linear, blendOutTime=0.1}
    boundBone {boneName=dyn_strap_small_base}
    componentAppliedLinearAccClamp=<1000 1000 1000>
    componentLinearAccScale=<0.3 0.3 0.3>
    componentLinearVelScale=<0.4 0.4 0.4>
    gravityOverride {z=-500}
    physicsBodyDefinitions [{
      collisionType=CustomSphere
      sphereCollisionRadius=2.5
      boundBone {boneName=dyn_strap_small_base}
      boxExtents=<8 4 6>
      constraintSetup {angularLimitsMax=<0 15 0>, angularLimitsMin=<0 -15 -40>, angularTarget=<0 0 0>}
      localJointOffset=<0.8 0 0>
     }]
    relativeSpaceBone {boneName=pack}
   }
   n30 = Anim Dynamics - Bone: dyn_strap_small_01(ComponentPose=@n29, bAlphaBoolEnabled=True, GravityOverride="0,0,-980", ExternalForce="0,0,0") {
    alphaInputType=Bool
    angularDampingOverride=1
    angularSpringConstant=1000
    bAngularSpring=true
    bOverrideAngularDamping=true
    bUseGravityOverride=true
    simulationSpace=RootRelative
    alphaBoolBlend {blendInTime=0.1, blendOption=Linear, blendOutTime=0.1}
    boundBone {boneName=dyn_strap_small_01}
    componentAppliedLinearAccClamp=<1000 1000 1000>
    componentLinearAccScale=<0.3 0.3 0.3>
    componentLinearVelScale=<0.4 0.4 0.4>
    gravityOverride {z=-500}
    physicsBodyDefinitions [{
      collisionType=CustomSphere
      sphereCollisionRadius=2.5
      boundBone {boneName=dyn_strap_small_01}
      boxExtents=<8 4 6>
      constraintSetup {angularLimitsMax=<0 22 0>, angularLimitsMin=<0 -22 -50>, angularTarget=<0 0 0>}
      localJointOffset=<0.8 0 0>
     }]
    relativeSpaceBone {boneName=pack}
   }
   Output Pose(Result=Blend Poses by bool(
     BlendPose_0=Component To Local(ComponentPose=@n30)
     BlendPose_1=Blend Poses by bool(BlendPose_0=@n24, BlendPose_1=@n17, bActiveValue=@n25) {blendType=Linear}
     bActiveValue=Get bIsInFrontEnd()
    ) {blendType=Linear}) {layerGroup=None}
   n31 = "Transform (Modify) Bone - Bone: pack"!(
    ComponentPose=Local To Component(LocalPose=Input Pose())
    Alpha:float
    Translation=Get backpack_translation()
    Rotation=Get backpack_rotate()
    Scale
   ) -> (Pose) {
    rotationMode=BMM_Additive
    scaleMode=BMM_Additive
    scaleSpace=BCS_BoneSpace
    translationMode=BMM_Additive
    alphaBoolBlend {blendOption=Linear}
    boneToModify {boneName=pack}
   }
   Save cached pose 'Base_pose'(Pose=Component To Local(ComponentPose=@n31)) {cacheName=Base_pose}
  ]}
 EventGraph {nodes [
   Event BlueprintUpdateAnimation()
   > n1 = Sequence()
   > Set inAir(inAir:bool=OR Boolean!(A=Get bIsSkydiveFloating(), B=Get bIsSkydiveDiveMode(), C:bool=Get bIsParachuteOpened()) -> (ReturnValue)) -> (Output_Get:bool)
   n2 = GetOwnerSkeletalMeshComponent()
   n3 = BreakRotator(InRot=BreakTransform(InTransform=GetDeltaTransformFromRefPose(self=@n2, BoneName=spine_05, BaseName=spine_01)).Rotation)
   n4 = MapRangeClamped(Value=@n3.Pitch, InRangeA=30, InRangeB=-30, OutRangeA=-30, OutRangeB=30)
   n5 = MapRangeClamped(Value=@n3.Pitch, InRangeA=0, InRangeB=-30, OutRangeA=5, OutRangeB=25)
   n6 = OR Boolean(A=Get bIsSkydiveFloating(), B=Get bIsSkyDiving(), C=Get bIsParachuteOpened())
   Set backpack_rotate(
    execute=@n1.then_1
    backpack_rotate=MakeRotator(Roll=MapRangeClamped(Value=@n3.Yaw, InRangeA=0, InRangeB=-30, OutRangeA=5, OutRangeB=25), Pitch=SelectFloat(A=@n5, B=@n4, bPickA=@n6))
   )
   > Set backpack_translation(backpack_translation=MakeVector(Y=MapRangeClamped(Value=@n3.Yaw, InRangeA=0, InRangeB=-30, OutRangeB=-6.5)))
   Event OnParachuteClosed()
   > Set Parashute Closed(Parashute Closed:bool=true) -> (Output_Get:bool)
   > Delay(Duration=3)
   > Set Parashute Closed(Parashute Closed=false)
   TryGetPawnOwner()
  ]}
]
variables [
 backpack_rotate {type=Rotator}
 backpack_translation {type=Vector}
 inAir {type=boolean}
 Parashute Closed {type=boolean}
]
